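; Server-driven manager for a Vivid Razor vault, running the alert from start through pause or opening.
; While alerted, a repeating timer derives progress from the door actor's weakpoint health, spawns weakpoints and opens the vault at full progress.
; State changes broadcast through gameplay event routers, while playspace tracking applies stacking participant-tag effects to players.
; It also reserves AI density, submits vault analytics and provides debug commands.
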(event Custom|Alert_Pause
  (if (ScalableFloat|GetValueasBoolwithContextParam (Variables|Default|GetVaultActor) (Variables|ScalableFloats|GetROW_Pause_Enabled))
    (Utilities|FlowControl|Sequence
      (:then_0
        (Class|BGAVividRazorVaultDoorParent|HandleAlertPaused (Variables|Default|GetVaultActor))
        (CallFunction|UpdateProgress 0.0)
        (CallFunction|CheckVaultShouldOpen))
      (:then_1
        (Utilities|FlowControl|DelayUntilNextTick)
        (Utilities|Time|ClearandInvalidateTimerbyHandle (Variables|Default|GetTimerHandle_AlertUpdate))
        (Variables|Alert|SetbAlert false)
        (Variables|Default|SetVaultPaused true)
        (CallFunction|StopListeningforPlayersinVaultSpace)
        (Variables|/VividRazorVaultGameplay/Gameplay/Environmental/Vault/BGA_VividRazorVault_DoorParent.BGA_VividRazorVault_DoorParent_C|SetbWeakpointIsExposed false (Variables|Default|GetVaultActor))
        (Networking|SwitchHasAuthority
          (:Authority
            (CallFunction|BroadcastPausedtoAllClients)
            (Analytics|SubmitAnalytics)))))))

(event EventOnWorldReady
  (|CallParentFunction)
  (Class|BGAVividRazorVaultDoorParent|ConnecttoManager (Variables|Default|GetVaultActor) self)
  (Utilities|TryGetEventRouter))

(event EventBeginPlay
  ; Event Begin Play
  (|CallParentFunction)
  (Networking|SwitchHasAuthority
    (:Authority
      (bind (_gamestate _playlist) (Class|FortAsyncActionCurrentPlaylistReady|PlaylistDataReadyAsync)
        (:OnReady
          (Utilities|IsValid _playlist
            (:Is%20Valid
              (Variables|AI|SetCurrentPlaylistMaxTeamSize (Variables|/Script/FortniteGame.FortPlaylist|GetMaxTeamSize _playlist)))))))))

(event EventEndPlay
  (params (EndPlayReason :type EEndPlayReason%20Enum :default "Destroyed"))
  ; Event End Play
  (|CallParentFunction EndPlayReason)
  (Analytics|SubmitAnalytics))

(event Custom|OnRequestStart
  (params (Payload :type Struct%20Vivid%20Razor%20Core%20Vault%20Start%20Payload%20Structure))
  ; On Request Start
  (Networking|SwitchHasAuthority
    (:Authority
      (if (not (Variables|Alert|GetbAlert))
        (Utilities|FlowControl|Sequence
          (:then_0
            (Variables|Alert|SetbAlert true)
            (Networking|ForceNetUpdate)
            (Variables|Analytics|SetStartTime (Fortnite|GetRealMatchTime))
            (Variables|Analytics|SetAnalyticsSubmitted false))
          (:then_1
            (bind _returnvalue (Utilities|Time|SetTimerbyEvent (EventDispatchers|CreateEvent :!event Custom|Alert_Update) 0.1 true true))
            (Variables|Default|SetTimerHandle_AlertUpdate _returnvalue)
            (if (not (Variables|Default|GetVaultPaused))
              (Variables|Default|SetSecondsRemaining (Variables|Default|GetTotalSecondsNeeded))))
          (:then_2
            (Variables|/VividRazorVaultGameplay/Gameplay/Environmental/Vault/BGA_VividRazorVault_DoorParent.BGA_VividRazorVault_DoorParent_C|SetVaultStarter (Pawn|GetController (Utilities|ResolveSoftReference (Utilities|Struct|BreakStructVividRazorCoreVaultStartPayload Payload))) (Variables|Default|GetVaultActor))
            (CallFunction|BroadcastAlerttoAllClients)
            (CallFunction|ListenforPlayersinVaultSpace)
            (bind _result (Utilities|Set|ToArray (Class|PlayspaceComponentSpatialActorTracker|GetTrackedActors (Variables|Default|GetPlayspaceSpatialActorTracker))))
            (for _array_element _result
              (bind _asfort_pawn (Utilities|Casting|CastToFortPawn _array_element))
              (ActivePlayers|AddParticipant (Variables|/Script/Engine.Pawn|GetPlayerState _asfort_pawn)))
            (Variables|Default|SetVaultPaused false)))))))

(event Collision|EventActorBeginOverlap
  (params (OtherActor :type Actor%20Object%20Reference)))

(event Custom|BroadcastVaultOpenedToAllClients
  ;[0.02083 0.02083 0.02083 0.5] Events
  ;; Opened
  (CallFunction|OnVaultOpened)
  (Utilities|TryGetEventRouter
    (:RouterValid
      (GameplayEventRouter|BroadcastStatefulEvent (Variables|Default|GetVaultEventRouter) (Utilities|Struct|MakeStructVividRazorCoreVaultOpenedPayload (Variables|Default|GetLocationIdTagConst) (Utilities|ResolveSoftReference self)) 0))))

(event Custom|BroadcastAlertToAllClients
  ;[0.02083 0.02083 0.02083 0.5] Events
  ;; Alert
  (Utilities|TryGetEventRouter
    (:RouterValid
      (GameplayEventRouter|BroadcastEvent (Variables|Default|GetVaultEventRouter) (Utilities|Struct|MakeStructVividRazorCoreVaultAlertPayload (Variables|Default|GetLocationIdTagConst) (Utilities|ResolveSoftReference self)) 0))))

(event EventTick
  (params (DeltaSeconds :type Float%20%28single-precision%29)))

(event Custom|DEBUG_VaultOpen
  ;[0.15 0.04219 0.04219 0.5] DEBUG Commands
  ;; cheat ke* DEBUG_VaultOpen
  (CallFunction|ForceOpenVault 0))

(event Custom|BroadcastWeakpointDestroyed_Client
  ;[0.02083 0.02083 0.02083 0.5] Events
  ;; Weakpoint Destroyed
  (Utilities|TryGetEventRouter
    (:RouterValid
      (GameplayEventRouter|BroadcastEvent (Variables|Default|GetVaultEventRouter) (Utilities|Struct|MakeStructVividRazorCoreVaultWeakpointDestroyedPayload (Variables|Default|GetLocationIdTagConst)) 0))))

(event Custom|BroadcastWeakpointDamaged_Client
  (params (DamageCauser :type Actor%20Soft%20Object%20Reference))
  ;[0.02083 0.02083 0.02083 0.5] Events
  ;; Weakpoint Damaged
  (Utilities|TryGetEventRouter
    (:RouterValid
      (GameplayEventRouter|BroadcastEvent (Variables|Default|GetVaultEventRouter) (Utilities|Struct|MakeStructVividRazorCoreVaultWeapointDamagedPayload (Variables|Default|GetLocationIdTagConst) DamageCauser) 0))))

(event Custom|DEBUG_VaultDestroyWeakpoint
  ;[0.15 0.04219 0.04219 0.5] DEBUG Commands
  ;; cheat ke* DEBUG_VaultDestroyWeakpoint
  (Class|BGAVividRazorVaultDoorParent|DestroyActiveWeakpoint (Variables|Default|GetVaultActor)))

(event Custom|ListenForPlayersInVaultSpace
  ; Notify Players Entering and Leaving Playspace
  (Networking|SwitchHasAuthority
    (:Authority
      (Utilities|TryGetEventRouter
        (:RouterValid
          (Variables|ActivePlayers|SetListeningForPlayers true)
          (bind (_asynctaskproxy _payload) (GameplayEventRouter|StartListeningToEvent (Variables|Default|GetVaultEventRouter) "/Script/FortniteGame.ActorTrackerActorEnteredContext")
            (:then
              (Variables|ActivePlayers|SetOnEnteredPlayspaceVolume _asynctaskproxy)
              (bind (_asynctaskproxy_2 _payload_2) (GameplayEventRouter|StartListeningToEvent (Variables|Default|GetVaultEventRouter) "/Script/FortniteGame.ActorTrackerActorExitContext")
                (:then
                  (Variables|ActivePlayers|SetOnExitedPlayspaceVolume _asynctaskproxy_2))
                (:OnEventReceived
                  (bind _aspawn (Utilities|Casting|CastToPawn (Utilities|Struct|BreakActorTrackerActorExitContext _payload_2)))
                  (ActivePlayers|RemoveParticipant (Variables|/Script/Engine.Pawn|GetPlayerState _aspawn)))))
            (:OnEventReceived
              (bind _aspawn_2 (Utilities|Casting|CastToPawn (Utilities|Struct|BreakActorTrackerActorEnteredContext _payload)))
              (ActivePlayers|AddParticipant (Variables|/Script/Engine.Pawn|GetPlayerState _aspawn_2)))))))))

(event Custom|DEBUG_VaultSpawnWeakpoint
  ;[0.15 0.04219 0.04219 0.5] DEBUG Commands
  ;; cheat ke* DEBUG_VaultSpawnWeakpoint
  (Class|BGAVividRazorVaultDoorParent|NewWeakpoint (Variables|Default|GetVaultActor))
  (Variables|Default|SetNumWeakpointsSpawned (+ (Variables|Default|GetNumWeakpointsSpawned) 1)))

(event Custom|BroadcastBossSet_Server
  ;[0.02083 0.02083 0.02083 0.5] Events
  ;; Boss Set to Spawn
  (Networking|SwitchHasAuthority
    (:Authority
      (Utilities|TryGetEventRouter
        (:RouterValid
          (GameplayEventRouter|BroadcastEvent (Variables|Default|GetVaultEventRouter) (Utilities|Struct|MakeStructVividRazorCoreVaultBossSetPlayload (Variables|Default|GetLocationIdTagConst) (Variables|Default|GetShowBossOnMinimap)) 0))))))

(event Custom|BroadcastPausedToAllClients
  ;[0.02083 0.02083 0.02083 0.5] Events
  ;; Paused
  (Utilities|TryGetEventRouter
    (:RouterValid
      (GameplayEventRouter|BroadcastEvent (Variables|Default|GetVaultEventRouter) (Utilities|Struct|MakeStructVividRazorCoreVaultPausedPayload (Variables|Default|GetLocationIdTagConst) self) 0))))

(event Custom|Start_Vault
  (params (Payload :type Struct%20Vivid%20Razor%20Core%20Vault%20Start%20Payload%20Structure))
  ; Objects and items that can start the process, interact directly with the vault, which then calls this function
  (CallFunction|OnRequestStart Payload))

(event Custom|StopListeningForPlayersInVaultSpace
  ; Notify Players Entering and Leaving Playspace
  (Networking|SwitchHasAuthority
    (:Authority
      (if (Variables|ActivePlayers|GetListeningForPlayers)
        (GameplayEventRouter|StopListeningtoEvent (Variables|ActivePlayers|GetOnEnteredPlayspaceVolume))
        (ActivePlayers|RemoveAllParticipants)
        (Variables|ActivePlayers|SetListeningForPlayers false)))))

(event Custom|AlertStatusReplicated
  ;[0.02083 0.02083 0.02083 0.5] Events
  ;; Alert Status Replicated
  (GameplayEventRouter|BroadcastEvent (GameplayEventRouter|GetGlobalEventRouter) (Utilities|Struct|MakeStructVividRazorCoreVaultAlertStatusReplicatedPayload (Variables|Default|GetLocationIdTagConst) (Utilities|ResolveSoftReference self)) 0))

(event Custom|Drill_Progress_Loop
  ; Drill Progress
  (CallFunction|HandleProgressChange 2.0))

(event Custom|Alert_Update
  ; Alert Update
  (CallFunction|HandleUpdate))

(event Custom|OnVaultOpened
  ; On Vault Opened
  (Analytics|SubmitAnalytics))

(event Custom|ResetVault
  (Utilities|Time|ClearandInvalidateTimerbyHandle (Variables|Default|GetTimerHandle_AlertUpdate))
  (Utilities|Time|ClearandInvalidateTimerbyHandle (Variables|Default|GetTimerHandle_DrillLoop))
  (Variables|Alert|SetbAlert false)
  (CallFunction|SetInitialWeakpoints)
  (Variables|Default|SetProgress 0.0))

(event Custom|ForceOpenVault
  (params (Instigator :type Controller%20Object%20Reference))
  (bind _ascontroller (Utilities|Casting|CastToController (Utilities|SelectObject (Variables|/VividRazorVaultGameplay/Gameplay/Environmental/Vault/BGA_VividRazorVault_DoorParent.BGA_VividRazorVault_DoorParent_C|GetVaultStarter (Variables|Default|GetVaultActor)) Instigator (Utilities|IsValid (Variables|/VividRazorVaultGameplay/Gameplay/Environmental/Vault/BGA_VividRazorVault_DoorParent.BGA_VividRazorVault_DoorParent_C|GetVaultStarter (Variables|Default|GetVaultActor)) :!class /Script/Engine.KismetSystemLibrary)))
    (:then
      _)
    (:CastFailed
      _))
  (Variables|/VividRazorVaultGameplay/Gameplay/Environmental/Vault/BGA_VividRazorVault_DoorParent.BGA_VividRazorVault_DoorParent_C|SetVaultStarter _ascontroller (Variables|Default|GetVaultActor))
  (Variables|Default|SetProgress 1.0)
  (CallFunction|CheckVaultShouldOpen)
  (Variables|Alert|SetbAlert true))

(event Custom|SetInitialWeakpoints
  (bind _vaultactor (Variables|Default|GetVaultActor))
  (bind _weakpointcount (Variables|/VividRazorVaultGameplay/Gameplay/Environmental/Vault/BGA_VividRazorVault_DoorParent.BGA_VividRazorVault_DoorParent_C|GetWeakpointCount _vaultactor))
  (bind _returnvalue (- 5 _weakpointcount))
  (Variables|Default|SetNumWeakpointsSpawned _returnvalue)
  (Variables|Default|SetNumWeakpointsDestroyed _returnvalue))

(event Custom|HandlePreconfiguredVaultVolumeSet
  (Networking|SwitchHasAuthority
    (:Remote
      (Utilities|IsValid (Variables|Default|GetPreconfiguredVaultVolume)
        (:Is%20Valid
          (Utilities|TryGetEventRouter))))))

(fn AddParticipant
  (params (PlayerState :type Player%20State%20Object%20Reference))
  (bind _returnvalue (Ability|GetAbilitySystemComponent PlayerState))
  (Utilities|IsValid _returnvalue
    (:Is%20Valid
      (GameplayEffects|ApplyGameplayEffectToSelf :self _returnvalue :GameplayEffectClass "/VividRazorVaultGameplay/Gameplay/GameplayEffects/GE_VividRazorVault_AddParticipantTag.GE_VividRazorVault_AddParticipantTag_C" :Level 0.0))))

(fn AddProgress_Drill)

(fn AddProgress_Weakpoint
  (Variables|Default|SetNumWeakpointsDestroyed (+ (Variables|Default|GetNumWeakpointsDestroyed) 1))
  (CallFunction|UpdateProgress 0.0))

(fn CheckForNewWeakpoints
  (if (< (Variables|Default|GetNumWeakpointsSpawned) (Math|Float|Truncate (ScalableFloat|GetValueatLevelwithContextParam (Variables|Default|GetVaultActor) (Variables|ScalableFloats|GetROW_Weakpoint_NumberToSpawn))))
    ; Is a weakpoint already alive? If so we don't want to spawn another (disabled for now)
    (if (not (Class|BGAVividRazorVaultDoorParent|IsWeakpointAlive (Variables|Default|GetVaultActor)))
      (Class|BGAVividRazorVaultDoorParent|NewWeakpoint (Variables|Default|GetVaultActor))
      (Variables|Default|SetNumWeakpointsSpawned (+ (Variables|Default|GetNumWeakpointsSpawned) 1)))))

(fn CheckVaultShouldOpen
  (if (>= (Variables|Default|GetProgress) 1.0)
    (Utilities|Time|ClearandInvalidateTimerbyHandle (Variables|Default|GetTimerHandle_AlertUpdate))
    (Utilities|Time|ClearandInvalidateTimerbyHandle (Variables|Default|GetTimerHandle_DrillLoop))
    (CallFunction|StopListeningforPlayersinVaultSpace)
    (CallFunction|BroadcastVaultOpenedtoAllClients)
    (bind _armoredtruckactor (Variables|Default|GetArmoredTruckActor)
      (:then
        (Class|BPIVividRazorCoreVaultEvents|VaultOpened(Message) _armoredtruckactor))
      (:else))))

(fn DisplayBossOnMinimap
  (params (ShouldShow :type Boolean))
  (Variables|Default|SetShowBossOnMinimap ShouldShow)
  (CallFunction|BroadcastBossSetServer))

(fn Get_bAlert
  (return (Variables|Alert|GetbAlert)))

(fn HandleProgressChange
  (params (NewProgress :type Float%20%28double-precision%29))
  (Utilities|FlowControl|Sequence
    (:then_0
      (CallFunction|UpdateProgress (Variables|GetNewProgress)))
    (:then_1
      (CallFunction|CheckforNewWeakpoints))
    (:then_2
      (CallFunction|CheckVaultShouldOpen))))

(fn HandleUpdate
  (if (Class|BGAVividRazorVaultDoorParent|IsEnabledandActive (Variables|Default|GetVaultActor))
    (Variables|Default|SetElapsedWeakPointTime (+ (Variables|Default|GetElapsedWeakPointTime) 0.1))
    (CallFunction|HandleProgressChange 0.0)))

(fn IsDrillAttached
  (return (Variables|Default|GetDrillAttached)))

(fn OnRep_bAlert
  (if (Variables|Alert|GetbAlert)
    (Networking|CanExecuteCosmeticEvents
      (:True
        (CallFunction|AlertStatusReplicated)))))

(fn OnRep_PreconfiguredVaultVolume
  (CallFunction|HandlePreconfiguredVaultVolumeSet))

(fn OnRep_Progress
  (Networking|CanExecuteCosmeticEvents
    (:True
      (if (== (Variables|Default|GetProgress) 0)
        (CallFunction|AlertStatusReplicated)))))

(fn RemoveAllParticipants
  (bind _result (Utilities|Set|ToArray (Class|PlayspaceComponentSpatialActorTracker|GetTrackedActors (Variables|Default|GetPlayspaceSpatialActorTracker))))
  (for _array_element _result
    (bind _aspawn (Utilities|Casting|CastToPawn _array_element))
    (ActivePlayers|RemoveParticipant (Variables|/Script/Engine.Pawn|GetPlayerState _aspawn))))

(fn RemoveParticipant
  (params (PlayerState :type Player%20State%20Object%20Reference))
  (bind _returnvalue (Ability|GetAbilitySystemComponent PlayerState))
  (Utilities|IsValid _returnvalue
    (:Is%20Valid
      (GameplayEffects|RemoveActiveGameplayEffectbySourceEffect _returnvalue "/VividRazorVaultGameplay/Gameplay/GameplayEffects/GE_VividRazorVault_AddParticipantTag.GE_VividRazorVault_AddParticipantTag_C" 0 1) ; Only remove 1 stack of the GE to support overlapping playspaces. Participant tag gets removed when all stacks are cleared
    )))

(fn ReserveAIDensityReservoir
  (params (bReserve :type Boolean))
  (bind _self self)
  (if (Variables|Default|GetShouldReserveAIDensity)
    (Networking|SwitchHasAuthority
      (:Authority
        (bind _livingworldmanager (Default|GetValidLivingWorldManager)
          (:Is%20Valid
            (if bReserve
              (Class|FortAthenaLivingWorldManager|RegisterDensityReservoir _livingworldmanager _self (Variables|AI|GetReservedReservoirDensity) (Variables|AI|GetReservedReservoirRange))
              (Fortnite|FortShippingLog _self "Registered density reservoir")
              (else
                (Class|FortAthenaLivingWorldManager|UnregisterDensityReservoir _livingworldmanager _self)
                (Fortnite|FortShippingLog _self "Unregistered self from density reservoir")))))))))

(fn SavePlayerWeakpointDamage
  (params (PlayerWeakpointDamage :type Float%20%28double-precision%29) (DamageType :type String))
  (Networking|SwitchHasAuthority
    (:Authority
      (Utilities|FlowControl|Sequence
        (:then_0
          (Utilities|Map|Add (Variables|Analytics|GetWeakpointDamage) (Variables|GetDamageType) (+ (Utilities|Map|Find (Variables|Analytics|GetWeakpointDamage) (Variables|GetDamageType)) (Variables|GetPlayerWeakpointDamage))))
        (:then_1
          (Variables|Analytics|SetAccruedPlayerWeakpointDamage (+ (Variables|Analytics|GetAccruedPlayerWeakpointDamage) (Variables|GetPlayerWeakpointDamage))))))))

(fn SubmitAnalytics
  (Networking|SwitchHasAuthority
    (:Authority
      (if (and (not (Replay|GetIsPlayingReplay)) (not (Variables|Analytics|GetAnalyticsSubmitted)))
        (Variables|Analytics|SetAnalyticsSubmitted true)
        (Utilities|FlowControl|Sequence
          (:then_0
            (Utilities|FlowControl|Sequence
              (:then_0
                (bind _returnvalue (Analytics|ConstructAnalyticsAttribute :AttributeName "VaultOpened" :IntValue -1 :floatValue -1.0 :BoolValue (>= (Variables|Default|GetProgress) 1.0)))
                (Utilities|Array|Add (LocalVariables|GetAnalyticsEventAttributes) _returnvalue)
                (bind _returnvalue_2 (Analytics|ConstructAnalyticsAttribute "VaultLocation" (Utilities|String|ToString(Vector) (Transformation|GetActorLocation))))
                (Utilities|Array|Add (LocalVariables|GetAnalyticsEventAttributes) _returnvalue_2)
                (bind _returnvalue_3 (Analytics|ConstructAnalyticsAttribute :AttributeName "VaultStartTime" :IntValue -1 :floatValue (Variables|Analytics|GetStartTime) :BoolValue false))
                (Utilities|Array|Add (LocalVariables|GetAnalyticsEventAttributes) _returnvalue_3)
                (bind _returnvalue_4 (Analytics|ConstructAnalyticsAttribute :AttributeName "VaultResultTime" :IntValue -1 :floatValue (Fortnite|GetRealMatchTime) :BoolValue false))
                (Utilities|Array|Add (LocalVariables|GetAnalyticsEventAttributes) _returnvalue_4))
              (:then_1
                (LocalVariables|SetWeakpointDamageArray "")
                (bind _keys (Utilities|Map|Keys (Variables|Analytics|GetWeakpointDamage)))
                (for _array_element _keys
                  (LocalVariables|SetWeakpointDamageArray (Utilities|String|Append (LocalVariables|GetWeakpointDamageArray) (select (Utilities|String|IsEmpty (LocalVariables|GetWeakpointDamageArray)) "" ",") _array_element ":" (Utilities|String|ToString(Float) (Utilities|Map|Find (Variables|Analytics|GetWeakpointDamage) _array_element)))))
                (bind _returnvalue_5 (Analytics|ConstructAnalyticsAttribute "VaultWeakpointDamageTypes" (LocalVariables|GetWeakpointDamageArray)))
                (Utilities|Array|Add (LocalVariables|GetAnalyticsEventAttributes) _returnvalue_5)
                (bind _returnvalue_6 (Analytics|ConstructAnalyticsAttribute :AttributeName "VaultWeakpointDamageTotal" :IntValue -1 :floatValue (Variables|Analytics|GetAccruedPlayerWeakpointDamage) :BoolValue false))
                (Utilities|Array|Add (LocalVariables|GetAnalyticsEventAttributes) _returnvalue_6)
                (bind _returnvalue_7 (Analytics|ConstructAnalyticsAttribute :AttributeName "VaultWeakpointsDestroyed" :IntValue (Variables|Default|GetNumWeakpointsDestroyed) :floatValue -1.0 :BoolValue false))
                (Utilities|Array|Add (LocalVariables|GetAnalyticsEventAttributes) _returnvalue_7))
              (:then_2
                (LocalVariables|SetVaultStarterID_String "INVALID")
                (bind _vaultactor (Variables|Default|GetVaultActor)
                  (:then
                    (bind _lastvaultstarter (Variables|/VividRazorVaultGameplay/Gameplay/Environmental/Vault/BGA_VividRazorVault_DoorParent.BGA_VividRazorVault_DoorParent_C|GetLastVaultStarter _vaultactor)
                      (:then
                        (bind _playerstate (Variables|/Script/Engine.Controller|GetPlayerState _lastvaultstarter)
                          (:then
                            (bind _asfort_player_state (Utilities|Casting|CastToFortPlayerState _playerstate)
                              (:then
                                (bind _returnvalue_8 (PlayerInfo|GetDebugStringforUniqueId (PlayerState|GetPlayerUniqueNetID _asfort_player_state)))
                                (LocalVariables|SetVaultStarterID_String _returnvalue_8))
                              (:CastFailed
                                _))
                            (bind _returnvalue_9 (Analytics|ConstructAnalyticsAttribute "VaultStarterID" (LocalVariables|GetVaultStarterID_String)))
                            (Utilities|Array|Add (LocalVariables|GetAnalyticsEventAttributes) _returnvalue_9))
                          (:else
                            _)))
                      (:else
                        _)))
                  (:else
                    _)))))
          (:then_1
            (Analytics|FireAnalyticsEventBlueprintArray "Core.VividRazor.Vault" (LocalVariables|GetAnalyticsEventAttributes))))))))

(fn UpdateProgress
  (params (AdditionalProgress :type Float%20%28double-precision%29))
  (Utilities|FlowControl|Sequence
    (:then_0
      (Variables|Default|SetSecondsRemaining (Math|Float|Clamp(Float) (- (Variables|Default|GetSecondsRemaining) (Variables|GetAdditionalProgress)) 0.0 (Variables|Default|GetTotalSecondsNeeded))))
    (:then_1
      (if (>= (Variables|Default|GetNumWeakpointsDestroyed) 5)
        (Variables|Default|SetProgress 1.0)
        (else
          (Variables|Default|SetProgress (* (- 1.0 (Class|VividRazorVaultDoor|GetTotalWeakPointHealthPercentage (Variables|Default|GetVaultActor))) (/ 5.0 (Variables|/VividRazorVaultGameplay/Gameplay/Environmental/Vault/BGA_VividRazorVault_DoorParent.BGA_VividRazorVault_DoorParent_C|GetWeakpointCount (Variables|Default|GetVaultActor))))))))))

(fn ConstructionScript)
