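; BP_VividRazorVaultPlayspace is a vault playspace that rates its relevance to a given player or actor.
; While the manager's vault is alerted and its progress is incomplete, GetPlayspaceRatingForContext adds a small score for proximity within MaxParticipantRange.
; If the player's reticle is on that vault's door, the score is set to 0.9, replacing the proximity value rather than adding to it.
; GetPlayspaceRatingForContext_old is the same logic without the alert check.
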(fn GetPlayspaceRatingForContext
  (params (WorldContextObject :type Object%20Reference))
  (Utilities|FlowControl|Sequence
    (:then_0
      (bind _asactor (Utilities|Casting|CastToActor (Variables|GetWorldContextObject)))
      (bind _vaultmanager (Variables|Default|GetVaultManager)
        (:then
          (Variables|/VividRazorVaultGameplay/Gameplay/Environmental/Manager/BGA_VividRazorVault_Manager.BGA_VividRazorVault_Manager_C|GetVaultActor _vaultmanager
            (:then
              (if (and (< (Variables|/VividRazorVaultGameplay/Gameplay/Environmental/Manager/BGA_VividRazorVault_Manager.BGA_VividRazorVault_Manager_C|GetProgress (Variables|Default|GetVaultManager)) 1.0) (Variables|/VividRazorVaultGameplay/Gameplay/Environmental/Manager/BGA_VividRazorVault_Manager.BGA_VividRazorVault_Manager_C|GetbAlert (Variables|Default|GetVaultManager)))
                (LocalVariables|SetScore (+ (LocalVariables|GetScore) (* (Math|Float|Clamp(Float) (/ (- (LocalVariables|GetMaxParticipantRange) (Math|Vector|VectorLength (- (Transformation|GetActorLocation (Variables|/VividRazorVaultGameplay/Gameplay/Environmental/Manager/BGA_VividRazorVault_Manager.BGA_VividRazorVault_Manager_C|GetVaultActor (Variables|Default|GetVaultManager))) (Transformation|GetActorLocation _asactor)))) (LocalVariables|GetMaxParticipantRange))) 0.1)))
                (bind _asfort_player_pawn (Utilities|Casting|CastToFortPlayerPawn (Variables|GetWorldContextObject)))
                (bind _asfort_player_controller (Utilities|Casting|CastToFortPlayerController (Pawn|GetController _asfort_player_pawn)))
                ; Value the vault that we are facing the highest
                (if (and (== (Utilities|Casting|CastToBGA_VividRazorVault_DoorParent (Reticle|GetActorUnderReticle _asfort_player_controller)) (Variables|/VividRazorVaultGameplay/Gameplay/Environmental/Manager/BGA_VividRazorVault_Manager.BGA_VividRazorVault_Manager_C|GetVaultActor (Variables|Default|GetVaultManager))) (.bSuccess (Utilities|Casting|CastToBGA_VividRazorVault_DoorParent (Reticle|GetActorUnderReticle _asfort_player_controller))))
                  (LocalVariables|SetScore 0.9) ; Looking at vault consists of 90%of score
                )))
            (:else)))
        (:else)))
    (:then_1
      (return (LocalVariables|GetScore)))))

(fn GetPlayspaceRatingForContext_old
  (params (WorldContextObject :type Object%20Reference))
  (Utilities|FlowControl|Sequence
    (:then_0
      (bind _asactor (Utilities|Casting|CastToActor (Variables|GetWorldContextObject)))
      (bind _vaultmanager (Variables|Default|GetVaultManager)
        (:then
          (Variables|/VividRazorVaultGameplay/Gameplay/Environmental/Manager/BGA_VividRazorVault_Manager.BGA_VividRazorVault_Manager_C|GetVaultActor _vaultmanager
            (:then
              (if (< (Variables|/VividRazorVaultGameplay/Gameplay/Environmental/Manager/BGA_VividRazorVault_Manager.BGA_VividRazorVault_Manager_C|GetProgress (Variables|Default|GetVaultManager)) 1.0)
                (LocalVariables|SetScore (+ (LocalVariables|GetScore) (* (Math|Float|Clamp(Float) (/ (- (LocalVariables|GetMaxParticipantRange) (Math|Vector|VectorLength (- (Transformation|GetActorLocation (Variables|/VividRazorVaultGameplay/Gameplay/Environmental/Manager/BGA_VividRazorVault_Manager.BGA_VividRazorVault_Manager_C|GetVaultActor (Variables|Default|GetVaultManager))) (Transformation|GetActorLocation _asactor)))) (LocalVariables|GetMaxParticipantRange))) 0.1)))
                (bind _asfort_player_pawn (Utilities|Casting|CastToFortPlayerPawn (Variables|GetWorldContextObject)))
                (bind _asfort_player_controller (Utilities|Casting|CastToFortPlayerController (Pawn|GetController _asfort_player_pawn)))
                ; Value the vault that we are facing the highest
                (if (and (== (Utilities|Casting|CastToBGA_VividRazorVault_DoorParent (Reticle|GetActorUnderReticle _asfort_player_controller)) (Variables|/VividRazorVaultGameplay/Gameplay/Environmental/Manager/BGA_VividRazorVault_Manager.BGA_VividRazorVault_Manager_C|GetVaultActor (Variables|Default|GetVaultManager))) (.bSuccess (Utilities|Casting|CastToBGA_VividRazorVault_DoorParent (Reticle|GetActorUnderReticle _asfort_player_controller))))
                  (LocalVariables|SetScore 0.9) ; Looking at vault consists of 90%of score
                )))
            (:else)))
        (:else)))
    (:then_1
      (return (LocalVariables|GetScore)))))

(fn ConstructionScript)
